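; Athena environment actor that builds a chain of deformed static meshes along an editable spline.
; Construction script fits the existing mesh to the first segment, then adds one per remaining segment, with optional roll, tension-based tangents and loop closure.
; Helpers apply material overrides, physical material and collision that blocks footstep-sound traces.
; Segments receive Index and LastIndex material parameters; ProgressRainbow pushes a Progress value to all of them.
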(event EventBeginPlay)

(event Collision|EventActorBeginOverlap
  (params (OtherActor :type Actor%20Object%20Reference)))

(event EventTick
  (params (DeltaSeconds :type Float%20%28single-precision%29)))

(event Custom|ProgressRainbow
  (for _array_element (Variables|Default|GetMeshes)
    (Rendering|Material|SetScalarParameterValueonMaterials _array_element "Progress" (Variables|Default|Get0-1%20Progress))))

(fn SetOverrideMaterials
  (params (StaticMeshComponent :type Spline%20Mesh%20Component%20Object%20Reference))
  (for (_array_element _array_index) (Variables|MaterialOverials|GetOverrideMaterials)
    (if (Utilities|IsValid _array_element :!class /Script/Engine.KismetSystemLibrary)
      (Rendering|Material|SetMaterial StaticMeshComponent _array_index _array_element))))

(fn SetPhysicalMaterial
  (params (PrimTarget :type Primitive%20Component%20Object%20Reference))
  (bind _physicalmatoverride (Variables|Default|GetPhysicalMatOverride))
  (Utilities|IsValid _physicalmatoverride
    (:Is%20Valid
      (Physics|SetPhysicalMaterialOverride PrimTarget _physicalmatoverride))
    (:Is%20Not%20Valid
      _))
  (return))

(fn SetSplineMeshCollision
  (params (SplineComponent :type Spline%20Mesh%20Component%20Object%20Reference))
  (if (Variables|Default|GetUseCollision)
    (Collision|SetCollisionEnabled SplineComponent "QueryAndPhysics")
    (Collision|SetCollisionResponsetoChannel SplineComponent "ECC_GameTraceChannel11" "ECR_Block") ; Block foot step traces for sound
    (else
      (Collision|SetCollisionEnabled SplineComponent "NoCollision"))))

(fn SetupSplineMesh
  (bind _returnvalue (AddComponent|NODE_AddSplineMeshComponent-2))
  (Components|StaticMesh|SetStaticMesh _returnvalue (Variables|Default|GetSplineStaticMesh))
  (SplineMesh|SetForwardAxis _returnvalue (Variables|Default|GetForwardAxis) true)
  (CallFunction|SetOverrideMaterials _returnvalue)
  (CallFunction|SetSplineMeshCollision _returnvalue)
  (Utilities|Array|Add (Variables|Default|GetMeshes) _returnvalue)
  (return _returnvalue))

(fn ConstructionScript
  (bind (_location _tangent) (Spline|GetLocationandTangentatSplinePoint (Variables|Default|GetSpline) 1 "Local"))
  (bind (_location_2 _tangent_2) (Spline|GetLocationandTangentatSplinePoint (Variables|Default|GetSpline) 0 "Local"))
  (bind _spline (Variables|Default|GetSpline))
  (bind (_location_3 _tangent_3) (Spline|GetLocationandTangentatSplinePoint _spline (- (Spline|GetNumberOfSplinePoints _spline) 1) "Local"))
  (bind (_location_4 _tangent_4) (Spline|GetLocationandTangentatSplinePoint _spline 0 "Local"))
  (bind _splinemesh (Variables|Default|GetSplineMesh))
  (bind _spline_2 (Variables|Default|GetSpline))
  (bind _splinemesh_2 (Variables|Default|GetSplineMesh))
  (bind _spline_3 (Variables|Default|GetSpline))
  (bind _splinemesh_3 (Variables|Default|GetSplineMesh))
  (bind _splinemesh_4 (Variables|Default|GetSplineMesh))
  (Components|StaticMesh|SetStaticMesh _splinemesh_3 (Variables|Default|GetSplineStaticMesh))
  (Utilities|Array|SetArrayElem (Variables|Default|GetMeshes) 0 _splinemesh_3 true)
  (CallFunction|SetPhysicalMaterial _splinemesh_4)
  (CallFunction|SetSplineMeshCollision _splinemesh_4)
  (SplineMesh|SetForwardAxis _splinemesh_4 (Variables|Default|GetForwardAxis) true)
  (SplineMesh|SetStartScale _splinemesh_2 (Math|Vector2D|MakeVector2D (.x (Spline|GetScaleatSplinePoint (Variables|Default|GetSpline) 0)) (.y (Spline|GetScaleatSplinePoint (Variables|Default|GetSpline) 0))) true :!class /Script/Engine.SplineMeshComponent)
  (SplineMesh|SetEndScale _splinemesh_2 (Math|Vector2D|MakeVector2D (.x (Spline|GetScaleatSplinePoint (Variables|Default|GetSpline) 1)) (.y (Spline|GetScaleatSplinePoint (Variables|Default|GetSpline) 1))) true :!class /Script/Engine.SplineMeshComponent)
  (SplineMesh|SetStartandEnd _splinemesh _location_2 _tangent_2 _location _tangent true :!class /Script/Engine.SplineMeshComponent)
  (Utilities|FlowControl|Sequence
    (:then_0
      (if (Variables|Default|GetEnableRoll)
        (SplineMesh|SetStartRollRadians _splinemesh (Math|Trig|DegreesToRadians (Spline|GetRollatSplinePoint _spline_2 0 "Local")) true)
        (SplineMesh|SetEndRollRadians _splinemesh (Math|Trig|DegreesToRadians (Spline|GetRollatSplinePoint _spline_2 1 "Local")) true)))
    (:then_1
      (for _index (range 1 (+ (- (Spline|GetNumberOfSplinePoints _spline_3) 2) 1))
        (bind _newparam (CallFunction|SetupSplineMesh))
        (SplineMesh|SetStartandEnd _newparam (Spline|GetLocationandTangentatSplinePoint _spline_3 _index "Local") (Math|Vector|SelectVector (* (Math|Vector|Normalize (.Tangent (Spline|GetLocationandTangentatSplinePoint _spline_3 _index "Local"))) (* (Math|Vector|VectorLength (- (Spline|GetLocationandTangentatSplinePoint _spline_3 (+ _index 1) "Local") (Spline|GetLocationandTangentatSplinePoint _spline_3 _index "Local"))) (Variables|Default|GetTension))) (.Tangent (Spline|GetLocationandTangentatSplinePoint _spline_3 _index "Local")) (Variables|Default|GetAutoSmoothTangent)) (Spline|GetLocationandTangentatSplinePoint _spline_3 (+ _index 1) "Local") (Math|Vector|SelectVector (* (Math|Vector|Normalize (.Tangent (Spline|GetLocationandTangentatSplinePoint _spline_3 (+ _index 1) "Local"))) (* (Math|Vector|VectorLength (- (Spline|GetLocationandTangentatSplinePoint _spline_3 (+ _index 1) "Local") (Spline|GetLocationandTangentatSplinePoint _spline_3 _index "Local"))) (Variables|Default|GetTension))) (.Tangent (Spline|GetLocationandTangentatSplinePoint _spline_3 (+ _index 1) "Local")) (Variables|Default|GetAutoSmoothTangent)) true :!class /Script/Engine.SplineMeshComponent)
        (CallFunction|SetPhysicalMaterial _newparam)
        (SplineMesh|SetStartScale _newparam (Math|Vector2D|MakeVector2D (.x (Spline|GetScaleatSplinePoint (Variables|Default|GetSpline) _index)) (.y (Spline|GetScaleatSplinePoint (Variables|Default|GetSpline) _index))) true :!class /Script/Engine.SplineMeshComponent)
        (SplineMesh|SetEndScale _newparam (Math|Vector2D|MakeVector2D (.x (Spline|GetScaleatSplinePoint (Variables|Default|GetSpline) (+ _index 1))) (.y (Spline|GetScaleatSplinePoint (Variables|Default|GetSpline) (+ _index 1)))) true :!class /Script/Engine.SplineMeshComponent)
        (if (Variables|Default|GetEnableRoll)
          (SplineMesh|SetStartRollRadians _newparam (Math|Trig|DegreesToRadians (Spline|GetRollatSplinePoint _spline_3 _index "Local")) true)
          (SplineMesh|SetEndRollRadians _newparam (Math|Trig|DegreesToRadians (Spline|GetRollatSplinePoint _spline_3 (+ _index 1) "Local")) true)))
      (Utilities|FlowControl|Sequence
        (:then_0
          (if (Variables|Default|GetClosedSpline)
            (Spline|SetClosedLoop _spline true true)
            (bind _newparam_2 (CallFunction|SetupSplineMesh))
            (SplineMesh|SetStartandEnd _newparam_2 _location_3 _tangent_3 _location_4 _tangent_4 true :!class /Script/Engine.SplineMeshComponent)))
        (:then_1
          (for (_array_element _array_index) (Variables|Default|GetMeshes)
            (Rendering|Material|SetScalarParameterValueonMaterials _array_element "Index" (Math|Conversions|ToFloat(Integer) _array_index))
            (Rendering|Material|SetScalarParameterValueonMaterials _array_element "LastIndex" (- (Math|Conversions|ToFloat(Integer) (Spline|GetNumberOfSplinePoints (Variables|Default|GetSpline))) 1.0))
            (CallFunction|SetOverrideMaterials _array_element)))))))
