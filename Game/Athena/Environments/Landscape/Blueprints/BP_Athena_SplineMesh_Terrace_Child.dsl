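; Terrace spline mesh child Blueprint that thins its rice instanced meshes at runtime by platform.
; On BeginPlay it sets a reduction factor for mobile or for non-Windows platforms with lower effects quality, leaving Windows at its default.
; After a random delay of up to one second, InitInstances keeps every Nth instance of Rice_02 and Rice_03, removing the rest.
; A custom test event reapplies the thinning with a given factor.
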(event EventBeginPlay
  (bind _instancereductionamount (Variables|Default|GetInstanceReductionAmount))
  (|CallParentFunction)
  (Utilities|FlowControl|Sequence
    (:then_0
      (if (TouchPlatform|IsMobileGame)
        (Variables|Default|SetInstanceReductionAmount 4)
        (elif (Utilities|String|Contains (Game|GetPlatformName) "Window")
          _
          (else
            (bind _returnvalue (Development|GetConsoleVariableIntValue "sg.EffectsQuality"))
            (if (< _returnvalue 3)
              (Variables|Default|SetInstanceReductionAmount 2) ; Xbox1/PS4
            )))))
    (:then_1
      (if (> (Variables|Default|GetInstanceReductionAmount) 0)
        (Utilities|FlowControl|Delay (Math|Random|RandomFloatinRange 0.0 1.0))
        (CallFunction|InitInstances (Variables|Default|GetHSIM_Rice_03) _instancereductionamount)
        (CallFunction|InitInstances (Variables|Default|GetHSIM_Rice_02) _instancereductionamount)))))

(event Custom|test_rice_reduction
  (params (count :type Integer))
  (bind _instancereductionamount (Variables|Default|GetInstanceReductionAmount))
  (Variables|Default|SetInstanceReductionAmount count)
  (CallFunction|InitInstances (Variables|Default|GetHSIM_Rice_03) _instancereductionamount)
  (CallFunction|InitInstances (Variables|Default|GetHSIM_Rice_02) _instancereductionamount))

(fn initInstances
  (params (Target :type Instanced%20Static%20Mesh%20Component%20Object%20Reference) (count :type Integer))
  (for _index (range (Components|InstancedStaticMesh|GetInstanceCount Target))
    (if (not (< (Math|Integer|%25%28Integer%29 _index count) 1))
      (Utilities|Array|Add (LocalVariables|Getl_instanceIndices) _index)))
  (Components|InstancedStaticMesh|RemoveInstances Target (LocalVariables|Getl_instanceIndices)))

(fn ConstructionScript
  (|CallParentFunction))
