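; Blueprint that styles named POI labels on the Athena minimap, setting colors and fonts by location state.
; BeginPlay binds refreshes to the game state's POI color and playlist-loaded delegates, plus replay scrub completion during replays.
; Each refresh applies snow styling to four tagged POIs, updates golden POI styling and colors all undiscovered locations.
; The authority-only calendar refresh, triggered on match ready, contains only an empty sequence.
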(event EventBeginPlay
  (bind _brgamestate (Variables|Default|GetBRGameState))
  ;[1 1 1] EVENT -> BeginPlay
  (Utilities|CheckandSetupGameState)
  (Game|BindEventtoUpdateNamedPoiColorDelegate _brgamestate (EventDispatchers|CreateEvent :!event Custom|UpdateNamedPoiColorDelegate_Bind))
  (Playlist|BindEventtoOnCurrentPlaylistLoaded _brgamestate (EventDispatchers|CreateEvent :!event Custom|OnCurrentPlaylistLoaded_Bind))
  (if (Replay|GetIsPlayingReplay)
    (ReplayContext|BindEventtoOnReplayScrubComplete (BlueprintContext|GetFortReplayContext) (EventDispatchers|CreateEvent :!event Custom|UpdateNamedPoiColorDelegate_Bind))
    (else
      _))
  (Class|FortAsyncActionAuthorityMatchReady|MatchReadyAsync
    (:then
      (CallFunction|RefreshMapLocations))
    (:OnReady
      (CallFunction|RefreshCalendar(AuthorityOnly)))))

(event Custom|RefreshCalendar%20%28Authority%20Only%29
  ;[1 1 1] CUSTOM EVENT -> RefreshCalendar (Authority Only)
  (Networking|SwitchHasAuthority
    (:Authority
      (Utilities|FlowControl|Sequence))))

(event Custom|RefreshMapLocations
  ;[1 1 1] CUSTOM EVENT -> RefreshMapLocations
  ;;[1 1 1] Update Minimap Text
  (Utilities|FlowControl|Sequence
    (:then_0
      (CallFunction|MakeLocationsVisibleinSnow (GameplayTags|MakeLiteralGameplayTagContainer "(GameplayTags=((TagName=\"Athena.Location.POI.PolarPeak\"),(TagName=\"Athena.Location.POI.FrostyFlights\"),(TagName=\"Athena.Location.POI.HappyHamlet\"),(TagName=\"Athena.Location.POI.ShiftyShafts\")))")))
    (:then_1
      (UI|Minimap|UpdateAllGoldenPoiColor (Variables|Default|GetBRGameState) (Variables|Default|GetGoldenPoiFontColor) (Variables|Default|GetGoldenPoiFont)))
    (:then_2
      (UI|Minimap|SetAllLocationsStateColor (Variables|Default|GetBRGameState) (Variables|Default|GetUndiscoveredPoiColor) "Undiscovered"))))

(event Custom|UpdateNamedPoiColorDelegate_Bind
  ;[1 1 1] CUSTOM EVENT -> UpdateNamedPoiColorDelegate_Bind
  (Utilities|CheckandSetupGameState)
  (CallFunction|RefreshMapLocations))

(event Custom|OnCurrentPlaylistLoaded_Bind
  (params (PlaylistName :type Name) (PlaylistContextTags :type Gameplay%20Tag%20Container%20Structure%20%28by%20ref%29))
  ;[1 1 1] CUSTOM EVENT -> OnCurrentPlaylistLoaded_Bind
  (Utilities|CheckandSetupGameState)
  (CallFunction|RefreshMapLocations))

(fn MakeLocationsVisibleInSnow
  (params (LocationTags :type Gameplay%20Tag%20Container%20Structure%20%28by%20ref%29))
  (bind _brgamestate (Variables|Default|GetBRGameState))
  (for _array_element (GameplayTags|BreakGameplayTagContainer LocationTags)
    (UI|Minimap|SetLocationColorinState _brgamestate _array_element (Variables|Default|GetSnowPoiFontColor) "Normal")
    (UI|Minimap|SetLocationFontinState _brgamestate _array_element (Variables|Default|GetSnowPoiFont) "Normal")))

(fn ConstructionScript)
